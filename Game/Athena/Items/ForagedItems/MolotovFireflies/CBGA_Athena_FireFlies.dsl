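; Foraged firefly-jar building actor that players can collect in the evening, with ambient Niagara and audio effects.
; Once Curie and playlist data are ready, the server turns it on and registers Evening and Morning callbacks that toggle IsEvening.
; OnRep_IsEvening sets interactability and visibility on the server, and switches particles and audio on clients.
; Interacting fades audio and briefly pulls fireflies toward the pawn's right hand via Niagara variables.
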(event Interaction|EventBlueprintOnBeginInteract
  (params (InteractingPawn :type Fort%20Pawn%20Object%20Reference) (InteractionBeingAttempted :type EInteractionBeingAttempted%20Enum :default "FirstInteraction") (InteractionType :type TInteractionType%20Enum :default "IT_NoInteraction"))
  ;[1 1 1] @Wirginia - This will fire when a player interacts and should return the location of their right hand
  (CallFunction|OnInteractAddNiagaraEffects 0 InteractingPawn))

(event Custom|OnInteractAddNiagaraEffects
  (params (SelfActor :type Building%20Actor%20Object%20Reference) (InteractingPawn :type Fort%20Pawn%20Object%20Reference))
  (bind _niagara (Variables|Default|GetNiagara))
  ;[1 1 1] @Wirginia - This will fire when a player interacts and should return the location of their right hand
  (CallFunction|AudioUpdate true)
  (Niagara|SetNiagaraVariable(Vector3) _niagara "Jar_Position" (+ (Transformation|GetSocketLocation (Variables|/Script/Engine.Character|GetMesh InteractingPawn) "Hand_R") "0, 0,37.000000"))
  (Niagara|SetNiagaraVariable(Bool) _niagara "Attract_To_Jar" true)
  (Utilities|Time|SetTimerbyEvent (delegate Custom|MakeFirefliesReturn) 2.0))

(event Custom|MakeFirefliesReturn
  ;[1 1 1] @Wirginia - This will fire when a player interacts and should return the location of their right hand
  (Niagara|SetNiagaraVariable(Bool) (Variables|Default|GetNiagara) "Attract_To_Jar" false))

(event EventBeginPlay
  (|CallParentFunction)
  (Networking|SwitchHasAuthority
    (:Authority
      ;[1 0.7573 0.3395] Is Curie Enabled?
      (Curie|OnCurieActivated
        (:OnCurieActive
          (Class|FortAsyncActionCurrentPlaylistReady|PlaylistDataReadyAsync
            (:OnReady
              ;[1 1 1] This is in the event that we load in and the server is already at night, when the objects should be turned off
              (CallFunction|TurnonForageableItems)
              ;[1 1 1] These bind the functions to the specific phase in the day / night cycle
              (TimeOfDay|RegisterDayPhaseHitCallback self "TurnOnForageableItems" "Evening" true)
              (TimeOfDay|RegisterDayPhaseHitCallback self "TurnOffForageableItems" "Morning" true))))))))

(event Custom|AudioUpdate
  (params (Is%20Collected? :type Boolean))
  (if Is%20Collected?
    (Audio|Components|Audio|FadeOut (Variables|Default|GetAmbientAudio) 1.0 0.0 "Logarithmic" :!class /Script/Engine.AudioComponent)))

(fn GetFocusedSocketLocation
  (return (+ (Transformation|GetWorldLocation (Variables|Default|GetStaticMesh)) "0, 0,75.000000")))

(fn OnRep_Collected)

(fn OnRep_IsEvening
  (bind _niagara (Variables|Default|GetNiagara))
  (bind _isevening (Variables|Default|GetIsEvening))
  (bind _ambientaudio (Variables|Default|GetAmbientAudio))
  ;[1 1 1] If we pass the "Is Evening" check - Turn on the particle effects, allow pawns to interact and unhide the actor
  (Networking|SwitchHasAuthority
    (:Authority
      (Variables|Interact|SetReplicatedCanInteract _isevening)
      (Rendering|SetActorHiddenInGame (not _isevening)))
    (:Remote
      (if (Variables|Default|GetIsEvening)
        (Components|Activation|Activate _niagara true)
        (Audio|Components|Audio|FadeIn _ambientaudio 3.0 1.0 0.0 "Linear" :!class /Script/Engine.AudioComponent)
        (else
          (Components|Activation|Deactivate _niagara)
          (Audio|Components|Audio|FadeOut _ambientaudio 3.0 0.0 "Linear" :!class /Script/Engine.AudioComponent))))))

(fn TurnOffForageableItems
  ;[1 1 1] When triggered, make sure that Is Evening is set to False, Turn off the Particles and Hide the actor in game
  (Variables|Default|SetIsEvening false))

(fn TurnOnForageableItems
  ;[1 1 1] When triggered set IsEvening to True and run the IsEvening logic
  (Variables|Default|SetIsEvening true))

(fn ConstructionScript
  (|CallParentFunction))
