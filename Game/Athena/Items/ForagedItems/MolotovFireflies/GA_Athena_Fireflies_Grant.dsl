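; Gameplay ability for the foraged fireflies item that gives the player Molotov Cocktail grenades.
; On activation it prints a "Granted" debug string, then validates the avatar actor and casts it to a Fort player pawn.
; It gives the pawn's controller the Molotov item in the NumberToGrant quantity, with player notification.
; The ability then ends, as it also does when validation or the cast fails.
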(event Ability|EventActivateAbility
  (bind _returnvalue (Ability|GetAvatarActorfromActorInfo))
  (Development|PrintString "Granted")
  (Utilities|IsValid _returnvalue
    (:Is%20Valid
      (bind _asfort_player_pawn (Utilities|Casting|CastToFortPlayerPawn _returnvalue)
        (:then
          (Inventory|GiveItemtoInventoryOwner :InventoryOwner (Fortnite|GetFortPlayerControllerfromActor _asfort_player_pawn) :ItemDefinition "/Game/Athena/Items/Consumables/MolotovCocktail/WID_Athena_Grenade_Molotov.WID_Athena_Grenade_Molotov" :NumberToGive (Variables|Default|GetNumberToGrant) :bNotifyPlayer true :ItemLevel -1 :PickupInstigatorHandle -1 :bUseItemPickupAnalyticEvent false :WeaponAmmoOverride -1)
          (Ability|EndAbility))
        (:CastFailed
          (Ability|EndAbility))))
    (:Is%20Not%20Valid
      (Ability|EndAbility))))

(event Ability|EventOnEndAbility
  (params (bWasCancelled :type Boolean)))
