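# Niagara effect for the foraged Molotov fireflies: a glowing, blinking swarm that can be drawn into a jar.
# FN_Fireflies_Optimized drives mesh/sprite flies with curl noise and point attraction, lerping toward User.Jar_Position and shrinking nearby when Attract_To_Jar is set.
# Firefly_Trail spawns ribbons by reading those particles, while FN_Glow_Light adds one glow sprite and flickering light.
# Scalability distance fraction widens wander and enlarges sprites; the older fireflies and fire-ribbon emitters are disabled.
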NiagaraSystem /Game/Athena/Items/ForagedItems/MolotovFireflies/NS_Fireflies
dependencies {
 usedDataInterfaces [
  &NiagaraDataInterfaceCamera
  &NiagaraDataInterfaceColorCurve
  &NiagaraDataInterfaceCurve
  &NiagaraDataInterfaceParticleRead
 ]
 usedDynamicInputs []
 usedModules [&SystemState]
}
emitters {
 FN_Fireflies {
  bEnabled=false
  emitterUpdateScript [
   &EmitterState()
   &SpawnRate(SpawnRate=0.5, Use Spawn Probability=false)
  ]
  particleSpawnScript [
   /Niagara/Modules/Spawn/Initialization/InitializeParticle(
    Lifetime=&UniformRangedFloat(Minimum=15, Maximum=20)
    Mesh Scale=&Multiply_VectorByFloat(Vector=<1 2 1>, Float=&UniformRangedFloat(Minimum=0.6, Maximum=0.7))
    Module.Write Color=true
    Module.Write Lifetime=true
    Module.Write Mass=true
    Module.Write Material Random=true
    Module.Write Position=false
    Module.Write Scale=true
    Module.Write SpriteRotation=false
    Module.Write SpriteSize=true
    Position=&SimulationPosition()
    Sprite Size=<65 65>
   )
   &DynamicMaterialParameters(
    Index 0 Param 1=10
    Index 0 Param 2=1
    Index 1 Param 1=&UniformRangedFloat(Minimum=0, Maximum=30)
    Index 1 Param 2=&UniformRangedFloat(Minimum=0, Maximum=1)
    Index 1 Param 3=&UniformRangedFloat(Minimum=0, Maximum=1)
    Index 1 Param 4=&UniformRangedFloat(Minimum=0, Maximum=1)
    Module.Param0WriteEnabled=true
    Module.Param0WriteEnabled1=true
    Module.Param1WriteEnabled=true
    Module.Param1WriteEnabled1=false
    Module.Param2WriteEnabled=true
    Module.Param2WriteEnabled1=false
    Module.Param3WriteEnabled=true
    Module.Param3WriteEnabled1=false
    Write Parameter Index 1=true
   )
   /Niagara/Modules/Spawn/Location/SphereLocation(
    Hemisphere Z=true
    MaskSpawnGroup=false
    Module.Override Randomness=false
    Non Uniform Scale=&VectorFromFloat(Value=&Lerp_Float(
      A=0
      B=1
      Alpha=&NormalizeDistanceRange(
       Start Position:NiagaraPosition=$Engine.Owner.Position
       End Position:NiagaraPosition=&ReturnCameraVector(Camera Property=::NewEnumerator0, Coordinate Space=::Simulation)
       Distance:NiagaraFloat=3000
       Invert Normalized Range:NiagaraBool=false
      )
     ))
    Offset=<0 0 60>
    Surface Only=false
   )
   /SetVariables(Particles.Random_Color:LinearColor=&ColorFromCurve(
     ColorCurve=DataInterface() {propertyValues {
       AlphaCurve {keys [<at 0 0.13 RCIM_Cubic>]}
       BlueCurve {keys [<at 0 0.2157 RCIM_Cubic>,<at 0.4434 0.1267>,<at 1 0.3013 RCIM_Cubic>]}
       GreenCurve {keys [<at 0 0.6312 RCIM_Cubic>,<at 0.4434 0.3072>,<at 1 0.4382 RCIM_Cubic>]}
       RedCurve {keys [<at 0 1 RCIM_Cubic>,<at 0.4434 1>,<at 1 1 RCIM_Cubic>]}
      }}
     CurveIndex=&UniformRangedFloat(Minimum=0, Maximum=1)
    )) {isSetParameters=true, module=SetVariables_#1}
  ]
  particleUpdateScript [
   &ParticleState()
   /Niagara/Modules/Update/Forces/CurlNoiseForce(Noise Frequency=0.025, Noise Strength=350, Pan Noise=true, Pan Noise Field=<0 0 1>)
   /Niagara/Modules/Update/Forces/PointAttractionForce(
    Attraction Radius=1000
    AttractionStrength:NiagaraFloat=&Lerp_Float(A=15, B=75, Alpha=&MakeCustomFloatFromBool(Bool=$User.Attract_To_Jar, True Float=1, False Float=0))
    AttractorPosition:Vector3f=&Lerp_Vector(
     A=&Add_Vector(
      A=/Niagara/DynamicInputs/Helpers/SimulationPosition()
      B=&MakeVector(
       X=&Lerp_Float(
        A=&UniformRangedFloat(Minimum=-250, Maximum=250)
        B=&UniformRangedFloat(Minimum=-800, Maximum=800)
        Alpha=&NormalizeDistanceRange(
         Start Position:NiagaraPosition=&SimulationPosition()
         End Position:NiagaraPosition=&ReturnCameraVector(Camera Property=::NewEnumerator0, Coordinate Space=::Simulation)
         Distance:NiagaraFloat=2000
         Invert Normalized Range:NiagaraBool=false
        )
       )
       Y=&Lerp_Float(
        A=&UniformRangedFloat(Minimum=-250, Maximum=250)
        B=&UniformRangedFloat(Minimum=-800, Maximum=800)
        Alpha=&NormalizeDistanceRange(
         Start Position:NiagaraPosition=&SimulationPosition()
         End Position:NiagaraPosition=&ReturnCameraVector(Camera Property=::NewEnumerator0, Coordinate Space=::Simulation)
         Distance:NiagaraFloat=2000
         Invert Normalized Range:NiagaraBool=false
        )
       )
       Z=&Lerp_Float(
        A=&UniformRangedFloat(Minimum=-190, Maximum=310)
        B=&UniformRangedFloat(Minimum=-300, Maximum=600)
        Alpha=&NormalizeDistanceRange(
         Start Position:NiagaraPosition=&SimulationPosition()
         End Position:NiagaraPosition=&ReturnCameraVector(Camera Property=::NewEnumerator0, Coordinate Space=::Simulation)
         Distance:NiagaraFloat=2000
         Invert Normalized Range:NiagaraBool=false
        )
       )
      )
     )
     B=$User.Jar_Position
     Alpha=&MakeCustomFloatFromBool(Bool=$User.Attract_To_Jar, True Float=1, False Float=0)
    )
    Module.Kill Within Radius=false
    Module.Use Falloff=true
   ) {module=PointAttractionForce002}
   &Drag(Use Linear Drag=true, Use Rotational Drag=false)
   &DynamicMaterialParameters(
    Index 0 Param 1=&UniformRangedFloat(Minimum=4, Maximum=5)
    Index 0 Param 2=&FloatFromCurve(FloatCurve=DataInterface() {propertyValues {Curve {keys [
         {interpMode=RCIM_Cubic, tangentMode=RCTM_User, time=0, value=0}
         {arriveTangent=0.002, interpMode=RCIM_Cubic, leaveTangent=0.002, tangentMode=RCTM_User, time=0.1827, value=1}
         {arriveTangent=-1.2195, interpMode=RCIM_Cubic, leaveTangent=-1.2195, tangentMode=RCTM_User, time=0.9833, value=0.6448}
         <at 1.0011 0.002>
        ]}}})
    Module.Param0WriteEnabled=false
    Module.Param1WriteEnabled=true
    Module.Param2WriteEnabled=false
    Module.Param3WriteEnabled=false
   ) {module=DynamicMaterialParameters001}
   &SolveForcesAndVelocity(Clamp Velocity=false, Limit Acceleration=false)
   &ScaleMeshSize(Scale Factor=&VectorFromFloat(Value=&FloatFromCurve(
      FloatCurve=DataInterface() {propertyValues {Curve {keys [
          {interpMode=RCIM_Cubic, tangentMode=RCTM_User, time=-0.0046, value=1.0113}
          {arriveTangent=-1.0067, interpMode=RCIM_Cubic, leaveTangent=-1.0067, tangentMode=RCTM_User, time=0.9283, value=0.7392}
          <at 1 0>
         ]}}}
      Scale Curve=2
     )))
   &ScaleMeshSize(Scale Factor=&VectorFromFloat(Value=&Lerp_Float(
      A=1
      B=&Multiply_Float(
       A=1
       B=&Lerp_Float(
        A=0
        B=1
        Alpha=&NormalizeDistanceRange(
         Start Position:NiagaraPosition=$Particles.Position
         End Position:NiagaraPosition=Linked() {linkedVariable {name=User.Jar_Position, type=Vector3f}}
         Distance:NiagaraFloat=300
         Invert Normalized Range:NiagaraBool=false
        )
       )
      )
      Alpha=&MakeCustomFloatFromBool(Bool=$User.Attract_To_Jar, True Float=1, False Float=0)
     ))) {module=ScaleMeshSize001}
   &CameraOffset(Camera Offset Amount=-50)
   &ScaleColor(
    Scale RGB=&MakeVectorFromLinearColorRGB(Color=&ColorFromCurve(ColorCurve=DataInterface() {propertyValues {
        AlphaCurve {keys [<at 0 1 RCIM_Cubic>]}
        BlueCurve {keys [
          <at 0.0076 0>
          {arriveTangent=60.2978, interpMode=RCIM_Cubic, leaveTangent=60.2978, time=0.0149, value=1}
          <at 0.0242 1>
          <at 0.0294 0>
          <at 0.1557 0>
          <at 0.1659 1>
          <at 0.1728 1>
          <at 0.1795 0>
          <at 0.2684 0>
          <at 0.281 1>
          <at 0.2871 1>
          <at 0.2905 0>
          <at 0.5293 0>
          <at 0.5344 1>
          <at 0.5394 1>
          <at 0.5486 0>
          <at 0.7196 0>
          <at 0.733 1>
          <at 0.7372 1>
          <at 0.7439 0>
          <at 0.9652 0>
          <at 0.9752 1>
          <at 0.9786 1>
          <at 0.988 0 RCIM_Cubic>
         ]}
        GreenCurve {keys [
          <at 0.0076 0>
          {arriveTangent=60.2978, interpMode=RCIM_Cubic, leaveTangent=60.2978, time=0.0149, value=1}
          <at 0.0242 1>
          <at 0.0294 0>
          <at 0.1557 0>
          <at 0.1659 1>
          <at 0.1728 1>
          <at 0.1795 0>
          <at 0.2684 0>
          <at 0.281 1>
          <at 0.2871 1>
          <at 0.2905 0>
          <at 0.5293 0>
          <at 0.5344 1>
          <at 0.5394 1>
          <at 0.5486 0>
          <at 0.7196 0>
          <at 0.733 1>
          <at 0.7372 1>
          <at 0.7439 0>
          <at 0.9652 0>
          <at 0.9752 1>
          <at 0.9786 1>
          <at 0.988 0 RCIM_Cubic>
         ]}
        RedCurve {keys [
          <at 0.0076 0>
          {arriveTangent=60.2978, interpMode=RCIM_Cubic, leaveTangent=60.2978, time=0.0149, value=1}
          <at 0.0242 1>
          <at 0.0294 0>
          <at 0.1557 0>
          <at 0.1659 1>
          <at 0.1728 1>
          <at 0.1795 0>
          <at 0.2684 0>
          <at 0.281 1>
          <at 0.2871 1>
          <at 0.2905 0>
          <at 0.5293 0>
          <at 0.5344 1>
          <at 0.5394 1>
          <at 0.5486 0>
          <at 0.7196 0>
          <at 0.733 1>
          <at 0.7372 1>
          <at 0.7439 0>
          <at 0.9652 0>
          <at 0.9752 1>
          <at 0.9786 1>
          <at 0.988 0 RCIM_Cubic>
         ]}
       }}))
    ScaleA=true
    ScaleRGB=true
   )
   /SetVariables(Particles.Sprite_Color:LinearColor=&MultiplyLinearColor(
     Linear Color=$Particles.Color
     Scale Factor=&MakeVector4_FromLinearColor(LinearColor=$Particles.Random_Color)
    )) {isSetParameters=true, module=SetVariables_#2}
   &ScaleSpriteSize(Scale Factor=&Vector2DFromFloat(Value=&Lerp_Float(
      A=1
      B=3
      Alpha=&NormalizeDistanceRange(
       Start Position:NiagaraPosition=&SimulationPosition()
       End Position:NiagaraPosition=&ReturnCameraVector(Camera Property=::NewEnumerator0, Coordinate Space=::Simulation)
       Distance:NiagaraFloat=3000
       Invert Normalized Range:NiagaraBool=false
      )
     )))
   &GenerateLocationEvent(Event Send Rate=2, Module.Use Event Send Rate=true) {module=GenerateLocationEvent001}
  ]
  rendererClasses [&NiagaraMeshRendererProperties, &NiagaraSpriteRendererProperties]
  renderers [
   {rendererClass=&NiagaraMeshRendererProperties, rendererIndex=0}
   {rendererClass=&NiagaraSpriteRendererProperties, rendererIndex=1}
  ]
 }
 FN_Fireflies_Optimized {
  emitterUpdateScript [
   &EmitterState(
    Enable Distance Culling=true
    Inactive Response=::NewEnumerator1
    Life Cycle Mode=::NewEnumerator1
    Loop Delay=&UniformRangedFloat(Minimum=0.5, Maximum=2.5)
    Loop Duration=&UniformRangedFloat(Minimum=0.6, Maximum=1.8)
    MaxDistance=6500
    Module.bUseMaxDistance=true
    Module.bUseMinDistance=true
    Recalculate Duration Each Loop=true
    Scalability Mode=::NewEnumerator1
    UseLoopDelay=true
   )
   &SpawnRate(SpawnRate=0.5, Use Spawn Probability=false)
  ]
  particleSpawnScript [
   &InitializeParticle(
    Lifetime=&UniformRangedFloat(Minimum=15, Maximum=20)
    Mesh Scale=&Multiply_VectorByFloat(Vector=<1 2 1>, Float=&UniformRangedFloat(Minimum=0.6, Maximum=0.7))
    Module.Write Color=true
    Module.Write Lifetime=true
    Module.Write Mass=true
    Module.Write Material Random=true
    Module.Write Position=false
    Module.Write Scale=true
    Module.Write SpriteRotation=false
    Module.Write SpriteSize=true
    Position=/Niagara/Functions/Localspace/SimulationPosition()
    Sprite Size=<65 65>
   )
   &DynamicMaterialParameters(
    Index 0 Param 1=10
    Index 0 Param 2=1
    Index 1 Param 1=&UniformRangedFloat(Minimum=0, Maximum=30)
    Index 1 Param 2=&UniformRangedFloat(Minimum=0, Maximum=1)
    Index 1 Param 3=&UniformRangedFloat(Minimum=0, Maximum=1)
    Index 1 Param 4=&UniformRangedFloat(Minimum=0, Maximum=1)
    Module.Param0WriteEnabled=true
    Module.Param0WriteEnabled1=true
    Module.Param1WriteEnabled=true
    Module.Param1WriteEnabled1=false
    Module.Param2WriteEnabled=true
    Module.Param2WriteEnabled1=false
    Module.Param3WriteEnabled=true
    Module.Param3WriteEnabled1=false
    Write Parameter Index 1=true
   )
   &SphereLocation(
    Hemisphere Z=true
    MaskSpawnGroup=false
    Module.Override Randomness=false
    Non Uniform Scale=&VectorFromFloat(Value=&Lerp_Float(A=0, B=1, Alpha=$Emitter.Scalability.DistanceFraction))
    Offset=<0 0 60>
    Surface Only=false
   )
   /SetVariables(Particles.Random_Color:LinearColor=&ColorFromCurve(
     ColorCurve=DataInterface() {propertyValues {
       AlphaCurve {keys [<at 0 0.13 RCIM_Cubic>]}
       BlueCurve {keys [<at 0 0.2157 RCIM_Cubic>,<at 0.4434 0.1267>,<at 1 0.3013 RCIM_Cubic>]}
       GreenCurve {keys [<at 0 0.6312 RCIM_Cubic>,<at 0.4434 0.3072>,<at 1 0.4382 RCIM_Cubic>]}
       RedCurve {keys [<at 0 1 RCIM_Cubic>,<at 0.4434 1>,<at 1 1 RCIM_Cubic>]}
      }}
     CurveIndex=&UniformRangedFloat(Minimum=0, Maximum=1)
    )) {isSetParameters=true, module=SetVariables_#1}
  ]
  particleUpdateScript [
   &ParticleState()
   &CurlNoiseForce(Noise Frequency=0.025, Noise Strength=350, Pan Noise=true, Pan Noise Field=<0 0 1>)
   &PointAttractionForce(
    Attraction Radius=1000
    AttractionStrength:NiagaraFloat=&Lerp_Float(A=15, B=75, Alpha=&MakeCustomFloatFromBool(Bool=$User.Attract_To_Jar, True Float=1, False Float=0))
    AttractorPosition:Vector3f=&Lerp_Vector(
     A=&Add_Vector(
      A=/Niagara/DynamicInputs/Helpers/SimulationPosition()
      B=&MakeVector(
       X=&Lerp_Float(
        A=&UniformRangedFloat(Minimum=-250, Maximum=250)
        B=&UniformRangedFloat(Minimum=-800, Maximum=800)
        Alpha=$Emitter.Scalability.DistanceFraction
       )
       Y=&Lerp_Float(
        A=&UniformRangedFloat(Minimum=-250, Maximum=250)
        B=&UniformRangedFloat(Minimum=-800, Maximum=800)
        Alpha=$Emitter.Scalability.DistanceFraction
       )
       Z=&Lerp_Float(
        A=&UniformRangedFloat(Minimum=-190, Maximum=310)
        B=&UniformRangedFloat(Minimum=-300, Maximum=600)
        Alpha=$Emitter.Scalability.DistanceFraction
       )
      )
     )
     B=$User.Jar_Position
     Alpha=&MakeCustomFloatFromBool(Bool=$User.Attract_To_Jar, True Float=1, False Float=0)
    )
    Module.Kill Within Radius=false
    Module.Use Falloff=true
   ) {module=PointAttractionForce002}
   &Drag(Use Linear Drag=true, Use Rotational Drag=false)
   &DynamicMaterialParameters(
    Index 0 Param 1=&UniformRangedFloat(Minimum=4, Maximum=5)
    Index 0 Param 2=&FloatFromCurve(FloatCurve=DataInterface() {propertyValues {Curve {keys [
         {interpMode=RCIM_Cubic, tangentMode=RCTM_User, time=0, value=0}
         {arriveTangent=0.002, interpMode=RCIM_Cubic, leaveTangent=0.002, tangentMode=RCTM_User, time=0.1827, value=1}
         {arriveTangent=-1.2195, interpMode=RCIM_Cubic, leaveTangent=-1.2195, tangentMode=RCTM_User, time=0.9833, value=0.6448}
         <at 1.0011 0.002>
        ]}}})
    Module.Param0WriteEnabled=false
    Module.Param1WriteEnabled=true
    Module.Param2WriteEnabled=false
    Module.Param3WriteEnabled=false
   ) {module=DynamicMaterialParameters001}
   &SolveForcesAndVelocity(Clamp Velocity=false, Limit Acceleration=false)
   &ScaleMeshSize(Scale Factor=&VectorFromFloat(Value=&FloatFromCurve(
      FloatCurve=DataInterface() {propertyValues {Curve {keys [
          {interpMode=RCIM_Cubic, tangentMode=RCTM_User, time=-0.0046, value=1.0113}
          {arriveTangent=-1.0067, interpMode=RCIM_Cubic, leaveTangent=-1.0067, tangentMode=RCTM_User, time=0.9283, value=0.7392}
          <at 1 0>
         ]}}}
      Scale Curve=2
     )))
   &ScaleMeshSize(Scale Factor=&VectorFromFloat(Value=&Lerp_Float(
      A=1
      B=&Lerp_Float(
       A=0
       B=1
       Alpha=&NormalizeDistanceRange(
        Start Position:NiagaraPosition=$Particles.Position
        End Position:NiagaraPosition=Linked() {linkedVariable {name=User.Jar_Position, type=Vector3f}}
        Distance:NiagaraFloat=300
        Invert Normalized Range:NiagaraBool=false
       )
      )
      Alpha=&MakeCustomFloatFromBool(Bool=$User.Attract_To_Jar, True Float=1, False Float=0)
     ))) {module=ScaleMeshSize001}
   &CameraOffset(Camera Offset Amount=-50)
   &ScaleColor(
    Scale RGB=&MakeVectorFromLinearColorRGB(Color=&ColorFromCurve(ColorCurve=DataInterface() {propertyValues {
        AlphaCurve {keys [<at 0 1 RCIM_Cubic>]}
        BlueCurve {keys [
          <at 0.0076 0>
          {arriveTangent=60.2978, interpMode=RCIM_Cubic, leaveTangent=60.2978, time=0.0149, value=1}
          <at 0.0242 1>
          <at 0.0294 0>
          <at 0.1557 0>
          <at 0.1659 1>
          <at 0.1728 1>
          <at 0.1795 0>
          <at 0.2684 0>
          <at 0.281 1>
          <at 0.2871 1>
          <at 0.2905 0>
          <at 0.5293 0>
          <at 0.5344 1>
          <at 0.5394 1>
          <at 0.5486 0>
          <at 0.7196 0>
          <at 0.733 1>
          <at 0.7372 1>
          <at 0.7439 0>
          <at 0.9652 0>
          <at 0.9752 1>
          <at 0.9786 1>
          <at 0.988 0 RCIM_Cubic>
         ]}
        GreenCurve {keys [
          <at 0.0076 0>
          {arriveTangent=60.2978, interpMode=RCIM_Cubic, leaveTangent=60.2978, time=0.0149, value=1}
          <at 0.0242 1>
          <at 0.0294 0>
          <at 0.1557 0>
          <at 0.1659 1>
          <at 0.1728 1>
          <at 0.1795 0>
          <at 0.2684 0>
          <at 0.281 1>
          <at 0.2871 1>
          <at 0.2905 0>
          <at 0.5293 0>
          <at 0.5344 1>
          <at 0.5394 1>
          <at 0.5486 0>
          <at 0.7196 0>
          <at 0.733 1>
          <at 0.7372 1>
          <at 0.7439 0>
          <at 0.9652 0>
          <at 0.9752 1>
          <at 0.9786 1>
          <at 0.988 0 RCIM_Cubic>
         ]}
        RedCurve {keys [
          <at 0.0076 0>
          {arriveTangent=60.2978, interpMode=RCIM_Cubic, leaveTangent=60.2978, time=0.0149, value=1}
          <at 0.0242 1>
          <at 0.0294 0>
          <at 0.1557 0>
          <at 0.1659 1>
          <at 0.1728 1>
          <at 0.1795 0>
          <at 0.2684 0>
          <at 0.281 1>
          <at 0.2871 1>
          <at 0.2905 0>
          <at 0.5293 0>
          <at 0.5344 1>
          <at 0.5394 1>
          <at 0.5486 0>
          <at 0.7196 0>
          <at 0.733 1>
          <at 0.7372 1>
          <at 0.7439 0>
          <at 0.9652 0>
          <at 0.9752 1>
          <at 0.9786 1>
          <at 0.988 0 RCIM_Cubic>
         ]}
       }}))
    ScaleA=true
    ScaleRGB=true
   )
   /SetVariables(Particles.Sprite_Color:LinearColor=&MultiplyLinearColor(
     Linear Color=$Particles.Color
     Scale Factor=&MakeVector4_FromLinearColor(LinearColor=$Particles.Random_Color)
    )) {isSetParameters=true, module=SetVariables_#2}
   &ScaleSpriteSize(Scale Factor=&Vector2DFromFloat(Value=&Lerp_Float(A=1, B=3, Alpha=$Emitter.Scalability.DistanceFraction)))
   &GenerateLocationEvent(Event Send Rate=2, Module.Use Event Send Rate=true) {enabled=false, module=GenerateLocationEvent001}
  ]
  rendererClasses [&NiagaraMeshRendererProperties, &NiagaraSpriteRendererProperties]
  renderers [
   {rendererClass=&NiagaraMeshRendererProperties, rendererIndex=0}
   {rendererClass=&NiagaraSpriteRendererProperties, rendererIndex=1}
  ]
 }
 FN_Glow_Light {
  emitterUpdateScript [
   &EmitterState(
    Enable Distance Culling=true
    Life Cycle Mode=::NewEnumerator1
    Loop Behavior=::NewEnumerator1
    Loop Duration Mode=::NewEnumerator1
    MaxDistance=20000
    Module.bUseMaxDistance=true
    Module.bUseMinDistance=true
    Reset Age On Awaken=true
    Scalability Mode=::NewEnumerator1
    UseLoopDelay=false
   )
   &SpawnBurst_Instantaneous(Spawn Count=1, Use Spawn Probability=false)
  ]
  particleSpawnScript [
   &InitializeParticle(
    Color=<rgba 1 0.2887 0.0778 0.025>
    Mass=&UniformRangedFloat(Minimum=0.6, Maximum=1.2)
    Module.Write Color=true
    Module.Write Lifetime=true
    Module.Write Mass=false
    Module.Write Material Random=true
    Module.Write Position=false
    Module.Write Scale=false
    Module.Write SpriteRotation=false
    Module.Write SpriteSize=true
    Position=/Niagara/Functions/Localspace/SimulationPosition()
    Sprite Size=<550 550>
   )
   &SystemLocation(MaskSpawnGroup=false, Offset=<0 0 50>)
   &DynamicMaterialParameters(
    Index 0 Param 1=0.5
    Index 0 Param 2=&UniformRangedFloat(Minimum=0.03, Maximum=0.1)
    Module.Param0WriteEnabled=true
    Module.Param1WriteEnabled=true
    Module.Param2WriteEnabled=true
    Module.Param3WriteEnabled=true
   )
  ]
  particleUpdateScript [
   &ParticleState(Kill Particles When Lifetime Has Elapsed=false, Let Infinitely Lived Particles Die When Emitter Deactivates=true)
   &ScaleColor(
    Scale Alpha=&FN_LOD_Distance_Remap(Start Distance=15000, End Distance=20000, Minimum Output=0.6, Maximum Output=0)
    ScaleA=true
    ScaleRGB=false
   )
   &SolveForcesAndVelocity(Clamp Velocity=false, Limit Acceleration=false)
   &ScaleSpriteSize(Scale Factor=&Vector2DFromFloat(Value=&Lerp_Float(A=0.6, B=1, Alpha=$Emitter.Scalability.DistanceFraction)))
   &SpriteRotationRate(Rotation Rate=50)
   /SetVariables(Particles.Light_Size:NiagaraFloat=&Lerp_Float(A=300, B=500, Alpha=$Emitter.Scalability.DistanceFraction)) {isSetParameters=true}
   /SetVariables(Particles.Light_Color:LinearColor=&MultiplyLinearColor(
     Linear Color=<rgba 1 0.1701 0.0937 1>
     Scale Factor=&MakeFloat4FromVectorAndFloat(
      "Vector (RGB)"=&VectorFromFloat(Value=&FloatFromCurve(
        FloatCurve=DataInterface() {propertyValues {Curve {keys [<at 0 1>,<at 0.0584 1.1216>,<at 0.1538 0.77>,<at 0.222 1.0327>,<at 0.3166 0.8798>,<at 0.3542 0.7265>,<at 0.4815 1.1358>,<at 0.5626 0.7055>,<at 0.6479 0.8407>,<at 0.737 1.0896>,<at 0.8795 0.7482>,<at 1 1>]}}}
        CurveIndex=$Emitter.NormalizedLoopAge
        Scale Curve=&Lerp_Float(
         A=0
         B=1
         Alpha=&FN_LOD_Distance_Remap(Start Distance=5000, End Distance=0, Minimum Output=0, Maximum Output=1)
        )
       ))
      "Float (Alpha)"=1
     )
    )) {isSetParameters=true}
   /SetVariables(Particles.LightViz:NiagaraInt32=&MakeIntFromFloat(Float=&Lerp_Float(
      A=0
      B=1
      Alpha=&FN_LOD_Distance_Remap(Start Distance=0, End Distance=5000, Minimum Output=0, Maximum Output=1)
     ))) {isSetParameters=true}
  ]
  rendererClasses [&NiagaraSpriteRendererProperties, &NiagaraLightRendererProperties]
  renderers [
   {rendererClass=&NiagaraSpriteRendererProperties}
   {rendererClass=&NiagaraLightRendererProperties, rendererIndex=1}
  ]
 }
 FN_Projectile_Ribbon_Fire {
  bEnabled=false
  emitterUpdateScript [&EmitterState()]
  particleSpawnScript [&InitializeRibbon(
    Lifetime=0.7
    Module.Write Color=true
    Module.Write Lifetime=true
    Module.Write Mass=true
    Module.Write Material Random=true
    Module.Write Position=false
    Module.Write RibbonTwist=false
    Module.Write RibbonWidth=true
    Ribbon Width=7
   )]
  particleUpdateScript [
   &ParticleState()
   &CurlNoiseForce(
    Noise Strength=&FloatFromCurve(FloatCurve=DataInterface() {propertyValues {Curve {keys [<at 0 1 RCIM_Cubic>,<at 1 0.5>]}}}, Scale Curve=2)
    Pan Noise=false
   )
   &SolveForcesAndVelocity(Clamp Velocity=false, Limit Acceleration=false)
   &ScaleRibbonWidth(Ribbon Width Scale=&Multiply_Float(
     A=&FloatFromCurve(FloatCurve=DataInterface() {propertyValues {Curve {keys [<at 0 1 RCIM_Cubic>,<at 1 0>]}}})
     B=&Lerp_Float(
      A=1.5
      B=3.5
      Alpha=&NormalizeDistanceRange(
       Start Position:NiagaraPosition=/Niagara/DynamicInputs/Helpers/SimulationPosition()
       End Position:NiagaraPosition=&ReturnCameraVector(Camera Property=::NewEnumerator0, Coordinate Space=::Simulation)
       Distance:NiagaraFloat=3000
       Invert Normalized Range:NiagaraBool=false
      )
     )
    ))
  ]
  rendererClasses [&NiagaraRibbonRendererProperties]
  renderers [{rendererClass=&NiagaraRibbonRendererProperties}]
 }
 Firefly_Trail {
  emitterUpdateScript [
   &EmitterState(Enable Distance Culling=true, MaxDistance=1500, Module.bUseMaxDistance=true, Module.bUseMinDistance=false, Scalability Mode=::NewEnumerator1)
   /SetVariables(Emitter.EmitterReader=DataInterface() {propertyValues {EmitterBinding {bindingMode=Other, emitterName=FN_Fireflies_Optimized}}}) {isSetParameters=true}
   &SpawnRate(
    SpawnRate=&self:Spawn_by_Emitter(New Particle Attribute Reader:NiagaraDataInterfaceParticleRead=$Emitter.EmitterReader, Rate:NiagaraFloat=20)
    Use Spawn Probability=true
   )
  ]
  particleSpawnScript [
   &self:Sample_Location(New Particle Attribute Reader:NiagaraDataInterfaceParticleRead=$Emitter.EmitterReader) {module=Sample_Location}
   &InitializeRibbon(
    Color=<rgba 100 1 1 1>
    Lifetime=0.7
    Module.Write Color=true
    Module.Write Lifetime=true
    Module.Write Mass=true
    Module.Write Material Random=true
    Module.Write Position=false
    Module.Write RibbonTwist=false
    Module.Write RibbonWidth=true
    Ribbon Width=/Game/Effects/Niagara/DynamicInputs/Math/FN_Fit(
     Clamp:NiagaraBool=false
     Value:NiagaraFloat=&self:SampleColor(Emitter:NiagaraDataInterfaceParticleRead=$Emitter.EmitterReader, SampleID:NiagaraInt32=$Particles.SampleID)
     SourceMin:NiagaraFloat=0
     SourceMax:NiagaraFloat=1
     DestMin:NiagaraFloat=0
     DestMax:NiagaraFloat=7
    )
   )
  ]
  particleUpdateScript [
   &ParticleState()
   &CurlNoiseForce(
    Noise Strength=&FloatFromCurve(FloatCurve=DataInterface() {propertyValues {Curve {keys [<at 0 1 RCIM_Cubic>,<at 1 0.5>]}}}, Scale Curve=2)
    Pan Noise=false
   ) {enabled=false}
   &SolveForcesAndVelocity(Clamp Velocity=false, Limit Acceleration=false)
   /SetVariables(Particles.RibbonLife:NiagaraFloat=&FN_Fit(
     Clamp:NiagaraBool=false
     Value:NiagaraFloat=&self:SampleColor(Emitter:NiagaraDataInterfaceParticleRead=$Emitter.EmitterReader, SampleID:NiagaraInt32=$Particles.SampleID)
     SourceMin:NiagaraFloat=0
     SourceMax:NiagaraFloat=1
     DestMin:NiagaraFloat=0.1
     DestMax:NiagaraFloat=0.7
    )) {isSetParameters=true}
   &ScaleColor(
    Scale Alpha=&FN_Fit(
     Clamp:NiagaraBool=false
     Value:NiagaraFloat=&self:SampleColor(Emitter:NiagaraDataInterfaceParticleRead=$Emitter.EmitterReader, SampleID:NiagaraInt32=$Particles.SampleID)
     SourceMin:NiagaraFloat=0
     SourceMax:NiagaraFloat=1
     DestMin:NiagaraFloat=0.3
     DestMax:NiagaraFloat=1
    )
    ScaleA=true
    ScaleRGB=false
   )
   &ScaleRibbonWidth(Ribbon Width Scale=&Multiply_Float(
     A=&FloatFromCurve(
      FloatCurve=DataInterface() {propertyValues {Curve {keys [<at 0 1 RCIM_Cubic>,<at 1 0>]}}}
      CurveIndex=&Divide_Float(A=$Particles.Age, B=$Particles.RibbonLife)
     )
     B=&Lerp_Float(A=1.5, B=3.5, Alpha=$Emitter.Scalability.DistanceFraction)
    ))
  ]
  rendererClasses [&NiagaraRibbonRendererProperties]
  renderers [{rendererClass=&NiagaraRibbonRendererProperties}]
 }
}
summary {
 systemName=NS_Fireflies
 emitters [
  {
   bEnabled=false
   emitterName=FN_Fireflies
   rendererClasses [&NiagaraMeshRendererProperties, &NiagaraSpriteRendererProperties]
  }
  {bEnabled=false, emitterName=FN_Projectile_Ribbon_Fire, rendererClasses [&NiagaraRibbonRendererProperties]}
  {emitterName=FN_Glow_Light, rendererClasses [&NiagaraSpriteRendererProperties, &NiagaraLightRendererProperties]}
  {
   emitterName=FN_Fireflies_Optimized
   rendererClasses [&NiagaraMeshRendererProperties, &NiagaraSpriteRendererProperties]
  }
  {emitterName=Firefly_Trail, rendererClasses [&NiagaraRibbonRendererProperties]}
 ]
 userVariables [
  User.Attract_To_Jar {type=NiagaraBool}
  User.Jar_Position {type=Vector3f, description="", defaultValue {struct=Vector3f, value=<0 0 0>}}
 ]
}
